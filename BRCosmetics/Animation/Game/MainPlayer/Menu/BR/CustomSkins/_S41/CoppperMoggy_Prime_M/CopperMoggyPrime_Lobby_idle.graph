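# Lobby idle Level Sequence for the CopperMoggyPrime custom skin, animating a FortPlayerMannequin through Control Rig tracks.
# MR_Biped_v02 drives the body, Fortnite_3L_Face_CtrlRig the face, and a BoneTweaker rig sits on the child Head mesh.
# Only face rig channels are listed, keying blinks, eye darts, brow shifts and asymmetric mouth poses.
# Face keys return to about their starting pose by frame 350, consistent with a looping idle.
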LevelSequence /BRCosmetics/Animation/Game/MainPlayer/Menu/BR/CustomSkins/_S41/CoppperMoggy_Prime_M/CopperMoggyPrime_Lobby_idle
bindings [
 Head {
  id=#1
  parent_id=#2
  possessed_class=SkeletalMeshComponentBudgeted
  tracks [{
    type=MovieSceneControlRigParameterTrack
    display_name=Fortnite_Face_BoneTweaker_CtrlRig
    sections [{type=MovieSceneControlRigParameterSection}]
   }]
 }
 FortPlayerMannequin {
  id=#2
  possessed_class=FortPlayerMannequin
  tracks [
   {type=MovieSceneSpawnTrack, sections [{type=MovieSceneSpawnSection, channels [None {type=MovieSceneScriptingBoolChannel, default=true}]}]}
   {
    type=MovieScene3DTransformTrack
    sections [{
      type=MovieScene3DTransformSection
      channels [
       Location.X {type=MovieSceneScriptingDoubleChannel, default=0, const [0,1]}
       Location.Y {type=MovieSceneScriptingDoubleChannel, default=0, const [0,1]}
       Location.Z {type=MovieSceneScriptingDoubleChannel, default=0, const [0,1]}
       Rotation.X {type=MovieSceneScriptingDoubleChannel, default=0, const [0,1]}
       Rotation.Y {type=MovieSceneScriptingDoubleChannel, default=0, const [0,1]}
       Rotation.Z {type=MovieSceneScriptingDoubleChannel, default=0, const [0,1]}
       Scale.X {type=MovieSceneScriptingDoubleChannel, default=1, const [1,1]}
       Scale.Y {type=MovieSceneScriptingDoubleChannel, default=1, const [1,1]}
       Scale.Z {type=MovieSceneScriptingDoubleChannel, default=1, const [1,1]}
      ]
     }]
   }
   {type=MovieSceneControlRigParameterTrack, display_name=MR_Biped_v02, sections [{type=MovieSceneControlRigParameterSection}]}
   {
    type=MovieSceneControlRigParameterTrack
    display_name=Fortnite_3L_Face_CtrlRig
    sections [{
      type=MovieSceneControlRigParameterSection
      channels [
       CTRL_C_eyesAim.Location.X {type=MovieSceneScriptingFloatChannel, default=0, const [0,17]}
       CTRL_C_eyesAim.Location.Y {type=MovieSceneScriptingFloatChannel, default=0, const [0,17]}
       CTRL_C_eyesAim.Location.Z {type=MovieSceneScriptingFloatChannel, default=0, const [0,17]}
       CTRL_C_eyesAim.Rotation.X {type=MovieSceneScriptingFloatChannel, default=0, const [0,17]}
       CTRL_C_eyesAim.Rotation.Y {type=MovieSceneScriptingFloatChannel, default=0, const [0,17]}
       CTRL_C_eyesAim.Rotation.Z {type=MovieSceneScriptingFloatChannel, default=0, const [0,17]}
       CTRL_C_eyesAim.Scale.X {type=MovieSceneScriptingFloatChannel, default=1, const [1,17]}
       CTRL_C_eyesAim.Scale.Y {type=MovieSceneScriptingFloatChannel, default=1, const [1,17]}
       CTRL_C_eyesAim.Scale.Z {type=MovieSceneScriptingFloatChannel, default=1, const [1,17]}
       CTRL_L_eyeAim.Location.X {type=MovieSceneScriptingFloatChannel, default=0, const [0,17]}
       CTRL_L_eyeAim.Location.Y {type=MovieSceneScriptingFloatChannel, default=0, const [0,17]}
       CTRL_L_eyeAim.Location.Z {type=MovieSceneScriptingFloatChannel, default=0, const [0,17]}
       CTRL_L_eyeAim.Rotation.X {type=MovieSceneScriptingFloatChannel, default=0, const [0,17]}
       CTRL_L_eyeAim.Rotation.Y {type=MovieSceneScriptingFloatChannel, default=0, const [0,17]}
       CTRL_L_eyeAim.Rotation.Z {type=MovieSceneScriptingFloatChannel, default=0, const [0,17]}
       CTRL_L_eyeAim.Scale.X {type=MovieSceneScriptingFloatChannel, default=1, const [1,17]}
       CTRL_L_eyeAim.Scale.Y {type=MovieSceneScriptingFloatChannel, default=1, const [1,17]}
       CTRL_L_eyeAim.Scale.Z {type=MovieSceneScriptingFloatChannel, default=1, const [1,17]}
       CTRL_R_eyeAim.Location.X {type=MovieSceneScriptingFloatChannel, default=0, const [0,17]}
       CTRL_R_eyeAim.Location.Y {type=MovieSceneScriptingFloatChannel, default=0, const [0,17]}
       CTRL_R_eyeAim.Location.Z {type=MovieSceneScriptingFloatChannel, default=0, const [0,17]}
       CTRL_R_eyeAim.Rotation.X {type=MovieSceneScriptingFloatChannel, default=0, const [0,17]}
       CTRL_R_eyeAim.Rotation.Y {type=MovieSceneScriptingFloatChannel, default=0, const [0,17]}
       CTRL_R_eyeAim.Rotation.Z {type=MovieSceneScriptingFloatChannel, default=0, const [0,17]}
       CTRL_R_eyeAim.Scale.X {type=MovieSceneScriptingFloatChannel, default=1, const [1,17]}
       CTRL_R_eyeAim.Scale.Y {type=MovieSceneScriptingFloatChannel, default=1, const [1,17]}
       CTRL_R_eyeAim.Scale.Z {type=MovieSceneScriptingFloatChannel, default=1, const [1,17]}
       CTRL_convergenceSwitch {type=MovieSceneScriptingFloatChannel, default=0, const [0,17]}
       CTRL_Master.Location.X {type=MovieSceneScriptingFloatChannel, default=32.9919, const [30.7919,27]}
       CTRL_Master.Location.Y {type=MovieSceneScriptingFloatChannel, default=0, const [0,17]}
       CTRL_Master.Location.Z {type=MovieSceneScriptingFloatChannel, default=0, const [-10.64,17]}
       CTRL_Master.Rotation.X {type=MovieSceneScriptingFloatChannel, default=0, const [0,17]}
       CTRL_Master.Rotation.Y {type=MovieSceneScriptingFloatChannel, default=0, const [0,17]}
       CTRL_Master.Rotation.Z {type=MovieSceneScriptingFloatChannel, default=0, const [0,17]}
       CTRL_Master.Scale.X {type=MovieSceneScriptingFloatChannel, default=1, thinned_from=17, keys [[0,1.7259],[350,1.7259]]}
       CTRL_Master.Scale.Y {type=MovieSceneScriptingFloatChannel, default=1, thinned_from=17, keys [[0,1.7259],[350,1.7259]]}
       CTRL_Master.Scale.Z {type=MovieSceneScriptingFloatChannel, default=1, thinned_from=17, keys [[0,1.7259],[350,1.7259]]}
       CTRL_C_jaw.X {
        type=MovieSceneScriptingFloatChannel
        default=0
        thinned_from=17
        keys [[0,-0.0999],[167,-0.1],[213,-0.0808],[250,-0.0808],[260,-0.0821],[270,-0.0892],[301,-0.0892],[325,-0.0999],[350,-0.0999]]
       }
       CTRL_C_jaw.Y {type=MovieSceneScriptingFloatChannel, default=0, thinned_from=17, keys [[0,0],[350,0]]}
       CTRL_C_jaw_openExtreme {type=MovieSceneScriptingFloatChannel, default=0, const [0,21]}
       CTRL_C_jaw_fwdBack {type=MovieSceneScriptingFloatChannel, default=0, const [0,21]}
       CTRL_R_jaw_ChinRaiseD {type=MovieSceneScriptingFloatChannel, default=0, const [0,21]}
       CTRL_R_jaw_ChinRaiseU {type=MovieSceneScriptingFloatChannel, default=0, const [0,21]}
       CTRL_L_jaw_ChinRaiseU {type=MovieSceneScriptingFloatChannel, default=0, const [0,21]}
       CTRL_L_jaw_ChinRaiseD {type=MovieSceneScriptingFloatChannel, default=0, const [0,21]}
       CTRL_R_mouth_lipsBlow {type=MovieSceneScriptingFloatChannel, default=0, const [0,21]}
       CTRL_L_mouth_lipsTogetherU {type=MovieSceneScriptingFloatChannel, default=0, const [0,21]}
       CTRL_L_mouth_lipsTogetherD {type=MovieSceneScriptingFloatChannel, default=0, const [0,21]}
       CTRL_R_mouth_lipsTogetherU {type=MovieSceneScriptingFloatChannel, default=0, const [0,21]}
       CTRL_L_mouth_lipsBlow {type=MovieSceneScriptingFloatChannel, default=0, const [0,21]}
       CTRL_R_mouth_lipsTogetherD {type=MovieSceneScriptingFloatChannel, default=0, const [0,21]}
       CTRL_C_tongue_roll.X {type=MovieSceneScriptingFloatChannel, default=0, const [0,21]}
       CTRL_C_tongue_roll.Y {type=MovieSceneScriptingFloatChannel, default=0, const [0,21]}
       CTRL_C_tongue.X {type=MovieSceneScriptingFloatChannel, default=0, const [0,21]}
       CTRL_C_tongue.Y {type=MovieSceneScriptingFloatChannel, default=0, const [0,21]}
       CTRL_C_tongue_narrowWide {type=MovieSceneScriptingFloatChannel, default=0, const [0,21]}
       CTRL_C_tongue_inOut {type=MovieSceneScriptingFloatChannel, default=0, const [0,21]}
       CTRL_R_nose.X {type=MovieSceneScriptingFloatChannel, default=0, thinned_from=19, keys [[0,0],[198,0],[213,0.317],[250,0],[350,0]]}
       CTRL_R_nose.Y {type=MovieSceneScriptingFloatChannel, default=0, thinned_from=19, keys [[0,0],[198,0],[213,0.0951],[250,0],[350,0]]}
       CTRL_L_nose.X {type=MovieSceneScriptingFloatChannel, default=0, thinned_from=19, keys [[0,0],[198,0],[213,0.317],[250,0],[350,0]]}
       CTRL_L_nose.Y {type=MovieSceneScriptingFloatChannel, default=0, thinned_from=19, keys [[0,0],[198,0],[213,0.0951],[250,0],[350,0]]}
       CTRL_L_eye.X {type=MovieSceneScriptingFloatChannel, default=0, const [0,32]}
       CTRL_L_eye.Y {
        type=MovieSceneScriptingFloatChannel
        default=0
        thinned_from=32
        keys [[0,-0.0003],[54,-0.0002],[64,0.0007],[71,0.0009],[193,0.0009],[198,-0.0003],[250,-0.0003],[253,0.0009],[325,0.0009],[340,-0.0003],[350,0.0009]]
       }
       CTRL_L_eye_blink {
        type=MovieSceneScriptingFloatChannel
        default=0
        thinned_from=54
        keys [[0,0],[31,0],[32,0.0282],[33,0.4645],[34,1],[36,0.3166],[37,0.1554],[45,-0.3952],[52,0],[60,0],[61,0.1439],[62,0.5048],[63,1],[64,0.7376],[65,0.351],[68,0.2042],[71,0.1436],[167,0.1436],[172,0.9772],[178,0],[183,-0.0039],[193,-0.2477],[194,-0.0897],[195,0.2139],[196,0.8301],[198,1],[200,0.6967],[201,0.4075],[202,0.0792],[205,0],[253,0],[254,0.1298],[255,0.3565],[256,0.8877],[258,1],[262,-0.0026],[270,-0.1946],[277,-0.2677],[298,-0.2677],[299,0.0609],[300,0.6361],[301,1],[305,0.0011],[321,-0.1639],[325,-0.1451],[340,0],[350,0]]
       }
       CTRL_R_eye.X {type=MovieSceneScriptingFloatChannel, default=0, const [0,32]}
       CTRL_R_eye.Y {type=MovieSceneScriptingFloatChannel, default=0, const [0,32]}
       CTRL_R_eye_blink {
        type=MovieSceneScriptingFloatChannel
        default=0
        thinned_from=54
        keys [[0,0],[31,0],[32,0.0419],[33,0.4765],[34,1],[36,0.3166],[37,0.1554],[45,-0.3963],[52,0],[60,0],[61,0.1439],[62,0.5048],[63,1],[64,0.7376],[65,0.351],[68,0.2042],[71,0.1897],[167,0.1897],[172,0.9772],[178,0],[183,-0.0046],[193,-0.2477],[194,-0.1179],[195,0.1662],[196,0.7979],[198,1],[200,0.6967],[201,0.4075],[202,0.0792],[205,0],[253,0],[254,0.1298],[255,0.3565],[256,0.8877],[258,1],[260,0.2593],[261,0],[270,-0.2021],[277,-0.2677],[298,-0.2677],[299,0.0609],[300,0.6361],[301,1],[303,0.2593],[304,0],[321,-0.164],[325,-0.1452],[340,0],[350,0]]
       }
       CTRL_C_eye.X {
        type=MovieSceneScriptingFloatChannel
        default=0
        thinned_from=56
        keys [[0,-0.3076],[63,-0.3076],[64,-0.451],[65,-0.6773],[66,-0.8045],[68,-0.8204],[71,-0.8326],[78,-0.6208],[122,-0.6208],[124,-0.4996],[148,-0.4996],[150,-0.6208],[167,-0.6208],[169,-0.5177],[172,0.0107],[178,0.5108],[181,0.6326],[183,0.4686],[184,0.3641],[193,-0.3076],[198,-0.3076],[205,-0.5471],[214,-0.5471],[216,-0.4481],[233,-0.4481],[235,-0.3083],[253,-0.3083],[258,-0.3409],[260,0.1842],[261,0.368],[299,0.368],[300,0.2792],[301,-0.0582],[304,-0.6798],[314,-0.4449],[321,-0.3076],[350,-0.3076]]
       }
       CTRL_C_eye.Y {
        type=MovieSceneScriptingFloatChannel
        default=0
        thinned_from=56
        keys [[0,0.1286],[32,0.1286],[33,0.1216],[34,-0.006],[36,0.1355],[37,0.2034],[45,0.1286],[52,0.1286],[54,0.1253],[61,0.0807],[62,0.0013],[63,-0.3895],[64,-0.2843],[65,-0.1168],[66,-0.0151],[68,-0.0605],[71,-0.1441],[78,-0.0413],[122,-0.0413],[124,-0.0171],[148,-0.0171],[150,-0.0413],[167,-0.0413],[169,-0.1382],[172,-0.5016],[178,-0.2024],[181,-0.0891],[183,-0.0696],[184,-0.07],[193,-0.0648],[198,-0.0359],[205,-0.039],[214,-0.039],[216,-0.0203],[233,-0.0203],[235,-0.0051],[253,-0.0051],[258,-0.2561],[260,0.0136],[261,-0.0536],[299,-0.0536],[300,-0.1722],[301,-0.4829],[304,-0.2311],[314,0.0144],[321,0.1184],[325,0.1215],[340,0.1286],[350,0.1286]]
       }
       CTRL_C_eyesConvergence {type=MovieSceneScriptingFloatChannel, default=0, const [0,32]}
       CTRL_L_eye_cheekRaise {
        type=MovieSceneScriptingFloatChannel
        default=0
        thinned_from=39
        keys [[0,0.7033],[32,0.7016],[34,0.6995],[35,0.7167],[37,0.7011],[193,0.7016],[198,0.8557],[206,0.7016],[213,0.7016],[215,0.7875],[224,0.7039],[243,0.7016],[250,0.7572],[253,0.8003],[258,0.8557],[267,0.8461],[270,0.783],[277,0.7016],[325,0.7016],[340,0.7033],[350,0.7016]]
       }
       CTRL_L_eye_squintInner {type=MovieSceneScriptingFloatChannel, default=0, thinned_from=33, keys [[0,0.5606],[37,0.5606],[45,0.5072],[52,0.5606],[350,0.5606]]}
       CTRL_L_nose_wrinkleUpper {type=MovieSceneScriptingFloatChannel, default=0, thinned_from=32, keys [[0,0],[31,0],[34,0.8609],[37,0],[350,0]]}
       CTRL_R_nose_wrinkleUpper {type=MovieSceneScriptingFloatChannel, default=0, thinned_from=32, keys [[0,0],[31,0],[34,0.8841],[37,0],[350,0]]}
       CTRL_R_eye_squintInner {type=MovieSceneScriptingFloatChannel, default=0, thinned_from=33, keys [[0,0.5606],[37,0.5606],[45,0.4983],[52,0.5606],[350,0.5606]]}
       CTRL_R_eye_cheekRaise {
        type=MovieSceneScriptingFloatChannel
        default=0
        thinned_from=38
        keys [[0,0.7033],[31,0.7016],[34,0.6995],[35,0.7167],[37,0.7011],[193,0.7016],[198,0.8557],[206,0.7016],[213,0.7016],[215,0.7875],[224,0.7039],[243,0.7016],[250,0.7572],[253,0.8003],[258,0.8557],[267,0.8461],[270,0.783],[277,0.7016],[325,0.7016],[340,0.7033],[350,0.7016]]
       }
       CTRL_C_eyelidsReduce {
        type=MovieSceneScriptingFloatChannel
        default=0
        thinned_from=20
        keys [[0,0.0193],[167,0.0193],[172,0.0151],[183,0.0001],[301,0.0001],[325,0.0148],[340,0.0193],[350,0.0193]]
       }
       CTRL_R_brow_lateral {
        type=MovieSceneScriptingFloatChannel
        default=0
        thinned_from=22
        keys [[0,0.312],[183,0.312],[198,0.5427],[301,0.5427],[325,0.3662],[340,0.312],[350,0.312]]
       }
       CTRL_L_brow_lateral {
        type=MovieSceneScriptingFloatChannel
        default=0
        thinned_from=22
        keys [[0,1],[183,1],[198,0.5752],[301,0.5752],[325,0.9002],[340,1],[350,1]]
       }
       CTRL_R_brow_down {type=MovieSceneScriptingFloatChannel, default=0, thinned_from=20, keys [[0,1],[90,1],[167,0.996],[183,1],[350,1]]}
       CTRL_L_brow_down {type=MovieSceneScriptingFloatChannel, default=0, thinned_from=20, keys [[0,1],[90,1],[167,0.996],[183,1],[350,1]]}
       CTRL_L_brow_raiseOut {
        type=MovieSceneScriptingFloatChannel
        default=0
        thinned_from=22
        keys [[0,0.0943],[90,0.0943],[167,0.1435],[172,0.1938],[183,0.8925],[198,0.2082],[206,0.2082],[213,0.1202],[270,0.1202],[301,0.1726],[325,0.1129],[340,0.0943],[350,0.0943]]
       }
       CTRL_L_brow_raiseIn {
        type=MovieSceneScriptingFloatChannel
        default=0
        thinned_from=21
        keys [[0,0.0943],[90,0.0943],[167,0.1883],[172,0.2418],[183,0.4424],[198,0.1211],[213,0.0332],[270,0.0332],[301,0.0855],[325,0.0925],[340,0.0943],[350,0.0943]]
       }
       CTRL_R_brow_raiseIn {
        type=MovieSceneScriptingFloatChannel
        default=0
        thinned_from=22
        keys [[0,0.0943],[64,0.0943],[68,0.2182],[78,0.3854],[90,0.3584],[110,0.2742],[172,0.2742],[183,0.448],[198,0.1268],[213,0.0388],[270,0.0388],[301,0.0911],[325,0.0938],[340,0.0943],[350,0.0942]]
       }
       CTRL_R_brow_raiseOut {
        type=MovieSceneScriptingFloatChannel
        default=0
        thinned_from=22
        keys [[0,0.0943],[64,0.0943],[68,0.2143],[78,0.3854],[90,0.3584],[110,0.2742],[172,0.2742],[183,0.3433],[198,0.1575],[213,0.0696],[270,0.0696],[301,0.1219],[325,0.101],[340,0.0943],[350,0.0943]]
       }
       CTRL_lookAtSwitch {type=MovieSceneScriptingFloatChannel, default=0, const [0,21]}
       CTRL_C_teethD.X {type=MovieSceneScriptingFloatChannel, default=0, const [0,21]}
       CTRL_C_teethD.Y {type=MovieSceneScriptingFloatChannel, default=0, const [0,21]}
       CTRL_C_teethU.X {type=MovieSceneScriptingFloatChannel, default=0, const [0,21]}
       CTRL_C_teethU.Y {type=MovieSceneScriptingFloatChannel, default=0, const [0,21]}
       CTRL_C_teeth_fwdBackU {type=MovieSceneScriptingFloatChannel, default=0, const [0,21]}
       CTRL_L_mouth_corner.X {
        type=MovieSceneScriptingFloatChannel
        default=0
        thinned_from=20
        keys [[0,-0.2545],[45,-0.2545],[54,-0.1986],[59,-0.1558],[64,-0.1262],[68,-0.1266],[78,-0.1285],[167,-0.1285],[172,-0.1289],[183,-0.2193],[206,-0.2835],[213,-0.2974],[250,-0.2974],[260,-0.2945],[270,-0.2904],[301,-0.2888],[325,-0.2545],[350,-0.2545]]
       }
       CTRL_L_mouth_corner.Y {
        type=MovieSceneScriptingFloatChannel
        default=0
        thinned_from=20
        keys [[0,-0.4666],[45,-0.4666],[54,-0.299],[59,-0.1705],[64,-0.0818],[68,-0.0829],[78,-0.0887],[167,-0.0887],[172,-0.0898],[183,-0.1259],[206,-0.3216],[213,-0.252],[250,-0.252],[260,-0.2665],[270,-0.2874],[301,-0.2951],[325,-0.4666],[350,-0.4666]]
       }
       CTRL_R_mouth_corner.X {
        type=MovieSceneScriptingFloatChannel
        default=0
        thinned_from=20
        keys [[0,0.0318],[45,0.0318],[54,0.2181],[59,0.3608],[64,0.4594],[68,0.4581],[78,0.4518],[167,0.4518],[172,0.4505],[183,0.1432],[206,0.5108],[213,0.5707],[250,0.5707],[260,0.5342],[270,0.3178],[301,0.2381],[325,0.0318],[350,0.0318]]
       }
       CTRL_R_mouth_corner.Y {
        type=MovieSceneScriptingFloatChannel
        default=0
        thinned_from=20
        keys [[0,-0.4594],[45,-0.4594],[54,-0.1613],[59,0.067],[64,0.2248],[68,0.2227],[78,0.2125],[167,0.2125],[172,0.2105],[183,0.084],[206,0.4566],[213,0.571],[250,0.571],[260,0.5012],[270,0.3555],[301,0.3018],[325,-0.4594],[350,-0.4594]]
       }
       CTRL_C_teeth_fwdBackD {type=MovieSceneScriptingFloatChannel, default=0, const [0,21]}
       CTRL_R_mouth_cornerSharpnessD {type=MovieSceneScriptingFloatChannel, default=0, thinned_from=17, keys [[0,0],[37,0],[45,0.0039],[54,0],[350,0]]}
       CTRL_R_mouth_cornerSharpnessU {type=MovieSceneScriptingFloatChannel, default=0, thinned_from=17, keys [[0,0],[37,0],[45,0.0231],[54,0],[350,0]]}
       CTRL_L_eye_eyelidU.X {type=MovieSceneScriptingFloatChannel, default=0, thinned_from=18, keys [[0,0],[172,0],[183,-1],[198,0],[350,0]]}
       CTRL_L_eye_eyelidU.Y {
        type=MovieSceneScriptingFloatChannel
        default=0
        thinned_from=18
        keys [[0,0],[45,0],[54,-0.0412],[64,-0.0412],[68,-0.0231],[78,0],[172,0],[183,-0.2061],[198,0],[350,0]]
       }
       CTRL_L_mouth_cornerSharpnessD {type=MovieSceneScriptingFloatChannel, default=0, thinned_from=17, keys [[0,0],[37,0],[45,0.027],[54,0],[350,0]]}
       CTRL_L_eye_eyelidD.X {type=MovieSceneScriptingFloatChannel, default=0, thinned_from=18, keys [[0,0],[350,0]]}
       CTRL_L_eye_eyelidD.Y {type=MovieSceneScriptingFloatChannel, default=0, thinned_from=18, keys [[0,0],[45,0],[54,-0.0824],[64,-0.0824],[68,-0.0463],[78,0],[350,0]]}
       CTRL_L_mouth_cornerSharpnessU {type=MovieSceneScriptingFloatChannel, default=0, thinned_from=17, keys [[0,0],[37,0],[45,0.0193],[54,0],[350,0]]}
       CTRL_R_mouth_lipsTowardsTeethD {type=MovieSceneScriptingFloatChannel, default=0, const [0,21]}
       CTRL_L_mouth_lipsTowardsTeethD {type=MovieSceneScriptingFloatChannel, default=0, const [0,21]}
       CTRL_R_eye_eyelidU.X {
        type=MovieSceneScriptingFloatChannel
        default=0
        thinned_from=18
        keys [[0,0],[45,0],[54,0.0412],[68,0.0411],[78,0.0397],[167,0.0397],[183,0.0396],[198,0],[350,0]]
       }
       CTRL_R_eye_eyelidU.Y {
        type=MovieSceneScriptingFloatChannel
        default=0
        thinned_from=18
        keys [[0,0],[45,0],[54,0.1236],[64,0.1236],[68,0.124],[78,0.1243],[167,0.1243],[172,0.1236],[183,0.1236],[198,0],[350,0]]
       }
       CTRL_L_mouth_lipsTowardsTeethU {type=MovieSceneScriptingFloatChannel, default=0, const [0,21]}
       CTRL_R_eye_eyelidD.X {type=MovieSceneScriptingFloatChannel, default=0, const [0,18]}
       CTRL_R_eye_eyelidD.Y {type=MovieSceneScriptingFloatChannel, default=0, thinned_from=18, keys [[0,0],[45,0],[54,0.0412],[64,0.0412],[68,0.0231],[78,0],[350,0]]}
       CTRL_R_mouth_lipsTowardsTeethU {type=MovieSceneScriptingFloatChannel, default=0, const [0,21]}
       CTRL_faceGUIfollowHead {type=MovieSceneScriptingFloatChannel, default=0, const [0,17]}
       CTRL_eyesAimFollowHead {type=MovieSceneScriptingFloatChannel, default=0, const [0,17]}
       CTRL_C_nose.X {type=MovieSceneScriptingFloatChannel, default=0, const [0,17]}
       CTRL_C_nose.Y {
        type=MovieSceneScriptingFloatChannel
        default=0
        thinned_from=17
        keys [[0,0],[45,0],[54,0.3409],[64,0.3396],[68,0.3385],[78,0.3335],[167,0.3335],[172,0.3325],[183,0.3248],[198,0],[350,0]]
       }
       CTRL_C_teeth_rotateU {type=MovieSceneScriptingFloatChannel, default=0, const [0,21]}
       CTRL_C_teeth_rotateD {type=MovieSceneScriptingFloatChannel, default=0, const [0,21]}
       CTRL_L_brow_inner {
        type=MovieSceneScriptingFloatChannel
        default=0
        thinned_from=24
        keys [[0,0],[90,0],[167,-0.0102],[172,-0.1005],[183,-0.6966],[198,-0.6966],[206,-0.423],[250,-0.423],[257,-0.5878],[270,-0.423],[295,-0.423],[301,-0.7802],[308,-0.3708],[325,-0.0993],[340,0],[350,0]]
       }
       CTRL_R_brow_inner {
        type=MovieSceneScriptingFloatChannel
        default=0
        thinned_from=24
        keys [[0,0],[90,0],[167,-0.0106],[172,-0.0924],[183,-0.6859],[198,-0.6859],[206,-0.3955],[250,-0.3955],[257,-0.5604],[270,-0.3955],[295,-0.3955],[301,-0.7527],[308,-0.3467],[325,-0.0929],[340,0],[350,0]]
       }
       CTRL_L_brow_outer {
        type=MovieSceneScriptingFloatChannel
        default=0
        thinned_from=24
        keys [[0,-0.0271],[90,-0.0271],[167,-0.0374],[172,-0.2373],[183,-0.1249],[198,-0.592],[206,0.1774],[250,0.1774],[257,0.0125],[270,0.1774],[295,0.1774],[301,-0.1798],[308,0.1522],[325,0.0209],[340,-0.0271],[350,-0.0271]]
       }
       CTRL_R_brow_outer {
        type=MovieSceneScriptingFloatChannel
        default=0
        thinned_from=24
        keys [[0,0.3756],[90,0.3756],[167,0.3597],[172,0.2082],[183,-0.6335],[198,-0.6335],[206,-0.0015],[250,-0.0015],[257,-0.1664],[270,-0.0015],[295,-0.0015],[301,-0.3587],[308,0.045],[325,0.287],[340,0.3756],[350,0.3756]]
       }
       CTRL_C_mouth.X {
        type=MovieSceneScriptingFloatChannel
        default=0
        thinned_from=17
        keys [[0,-0.2853],[167,-0.2853],[213,-0.0633],[250,-0.0633],[260,-0.0657],[270,-0.1187],[301,-0.1187],[325,-0.2853],[350,-0.2853]]
       }
       CTRL_C_mouth.Y {
        type=MovieSceneScriptingFloatChannel
        default=0
        thinned_from=17
        keys [[0,0.107],[45,0.107],[54,0.1209],[59,0.1367],[68,0.1704],[167,0.1704],[213,0.1081],[250,0.1081],[260,0.1038],[270,0.0904],[301,0.0904],[325,0.107],[350,0.107]]
       }
       CTRL_L_mouth_upperLipRaise {type=MovieSceneScriptingFloatChannel, default=0, const [0,17]}
       CTRL_L_mouth_cornerPull {
        type=MovieSceneScriptingFloatChannel
        default=0
        thinned_from=17
        keys [[0,0],[167,0],[213,0.2114],[250,0.2114],[260,0.1971],[270,0.1188],[301,0.1188],[325,0],[350,0]]
       }
       CTRL_L_mouth_dimple {
        type=MovieSceneScriptingFloatChannel
        default=0
        thinned_from=17
        keys [[0,0],[45,0],[54,0.1596],[59,0.2693],[64,0.3425],[68,0.3415],[78,0.3364],[167,0.3364],[213,0.2114],[250,0.2114],[260,0.1971],[270,0.1188],[301,0.1188],[325,0],[350,0]]
       }
       CTRL_L_mouth_cornerDepress {
        type=MovieSceneScriptingFloatChannel
        default=0
        thinned_from=17
        keys [[0,0],[45,0],[54,0.0306],[59,0.0517],[64,0.0657],[68,0.0656],[78,0.0648],[167,0.0648],[213,0.2114],[250,0.2114],[260,0.1971],[270,0.1188],[301,0.1188],[325,0],[350,0]]
       }
       CTRL_L_mouth_stretch {type=MovieSceneScriptingFloatChannel, default=0, const [0,17]}
       CTRL_L_mouth_lowerLipDepress {type=MovieSceneScriptingFloatChannel, default=0, const [0,17]}
       CTRL_R_mouth_lowerLipDepress {type=MovieSceneScriptingFloatChannel, default=0, const [0,17]}
       CTRL_R_mouth_stretch {type=MovieSceneScriptingFloatChannel, default=0, const [0,17]}
       CTRL_L_mouth_towardsD {type=MovieSceneScriptingFloatChannel, default=0, const [0,21]}
       CTRL_R_mouth_funnelU {type=MovieSceneScriptingFloatChannel, default=0, const [0,21]}
       CTRL_L_mouth_towardsU {type=MovieSceneScriptingFloatChannel, default=0, const [0,21]}
       CTRL_R_mouth_lipBiteD {type=MovieSceneScriptingFloatChannel, default=0, const [0,21]}
       CTRL_R_mouth_funnelD {type=MovieSceneScriptingFloatChannel, default=0, const [0,21]}
       CTRL_L_mouth_pressU {
        type=MovieSceneScriptingFloatChannel
        default=0
        thinned_from=20
        keys [[0,0],[45,0],[54,0.2146],[64,0.2141],[68,0.2134],[78,0.2102],[167,0.2102],[172,0.2096],[183,0.2096],[206,0],[350,0]]
       }
       CTRL_L_mouth_funnelD {type=MovieSceneScriptingFloatChannel, default=0, const [0,21]}
       CTRL_L_mouth_suckBlow {type=MovieSceneScriptingFloatChannel, default=0, const [0,21]}
       CTRL_L_mouth_pressD {
        type=MovieSceneScriptingFloatChannel
        default=0
        thinned_from=20
        keys [[0,0],[45,0],[54,0.2409],[64,0.2402],[68,0.2395],[78,0.2359],[167,0.2359],[172,0.2352],[183,0.2352],[206,0],[350,0]]
       }
       CTRL_L_mouth_funnelU {type=MovieSceneScriptingFloatChannel, default=0, const [0,21]}
       CTRL_R_mouth_towardsU {type=MovieSceneScriptingFloatChannel, default=0, const [0,21]}
       CTRL_R_mouth_purseD {type=MovieSceneScriptingFloatChannel, default=0, const [0,21]}
       CTRL_L_mouth_lipBiteU {type=MovieSceneScriptingFloatChannel, default=0, const [0,21]}
       CTRL_R_mouth_pressD {
        type=MovieSceneScriptingFloatChannel
        default=0
        thinned_from=20
        keys [[0,0],[45,0],[54,0.2602],[64,0.2595],[68,0.2587],[78,0.2549],[167,0.2549],[172,0.2541],[183,0.2541],[206,0],[350,0]]
       }
       CTRL_R_mouth_pressU {
        type=MovieSceneScriptingFloatChannel
        default=0
        thinned_from=20
        keys [[0,0],[45,0],[54,0.2022],[64,0.2017],[68,0.2011],[78,0.1981],[167,0.1981],[172,0.1975],[183,0.1975],[206,0],[350,0]]
       }
       CTRL_L_mouth_purseD {type=MovieSceneScriptingFloatChannel, default=0, const [0,21]}
       CTRL_R_mouth_lipBiteU {type=MovieSceneScriptingFloatChannel, default=0, const [0,21]}
       CTRL_R_mouth_suckBlow {type=MovieSceneScriptingFloatChannel, default=0, const [0,21]}
       CTRL_L_mouth_purseU {type=MovieSceneScriptingFloatChannel, default=0, const [0,21]}
       CTRL_R_mouth_towardsD {type=MovieSceneScriptingFloatChannel, default=0, const [0,21]}
       CTRL_L_mouth_lipBiteD {type=MovieSceneScriptingFloatChannel, default=0, const [0,21]}
       CTRL_R_mouth_purseU {type=MovieSceneScriptingFloatChannel, default=0, const [0,21]}
       CTRL_R_mouth_cornerDepress {
        type=MovieSceneScriptingFloatChannel
        default=0
        thinned_from=17
        keys [[0,0],[45,0],[54,0.0411],[59,0.0693],[64,0.0881],[68,0.0879],[78,0.0865],[167,0.0865],[213,0],[350,0]]
       }
       CTRL_R_mouth_upperLipRaise {type=MovieSceneScriptingFloatChannel, default=0, const [0,17]}
       CTRL_R_mouth_dimple {
        type=MovieSceneScriptingFloatChannel
        default=0
        thinned_from=17
        keys [[0,0],[45,0],[54,0.2902],[59,0.4896],[64,0.6227],[68,0.6208],[78,0.6115],[167,0.6115],[213,0.4075],[250,0.4075],[260,0.3799],[270,0.2289],[301,0.2289],[325,0],[350,0]]
       }
       CTRL_R_mouth_cornerPull {
        type=MovieSceneScriptingFloatChannel
        default=0
        thinned_from=17
        keys [[0,0],[45,0],[54,0.0441],[59,0.0744],[64,0.0946],[68,0.0943],[78,0.0929],[167,0.0929],[213,0.1229],[250,0.1229],[260,0.1145],[270,0.069],[301,0.069],[325,0],[350,0]]
       }
      ]
     }]
   }
  ]
 }
]
display_rate {denominator=1, numerator=30}
tick_resolution {denominator=1, numerator=24000}
tracks []
